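# Physics asset for the CometHoliday pickaxe, setting up simulation bodies for its dangling strings, ribbons and bows.
# Each bone gets one capsule, and the long and short string bases have all angular motion locked to b_AR_Root.
# Two three-segment string chains hang from those bases, with both ribbons and both bows constrained to the short string's last segment.
# The long string's tip is also constrained to that segment, closing the two chains into a loop.
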PhysicsAsset /BRCosmetics/Weapons/FORT_Melee/Pickaxe_CometHoliday/Meshes/CometHoliday_Axe_Physics
bodies {
 b_AR_Root [{shapeType=Capsule, length=0.505, radius=0.505, shapeName=b_AR_Root_capsule, center=<2 0 0>, rotation=<rot 0 -90 0>}]
 base_longString_00 [{shapeType=Capsule, length=0.505, radius=0.505, shapeName=base_longString_00_capsule, center=<2 0 0>, rotation=<rot 0 -90 0>}]
 base_shortString_00 [{shapeType=Capsule, length=0.505, radius=0.505, shapeName=base_shortString_00_capsule, rotation=<rot 0 -90 0>}]
 dyn_backRibbon_00 [{shapeType=Capsule, length=1.7159, radius=1.5744, shapeName=dyn_backRibbon_00_capsule, center=<2 0 0>, rotation=<rot 0 -3.56 -85.3675>}]
 dyn_frontRibbon_00 [{shapeType=Capsule, length=1.8439, radius=1.6248, shapeName=dyn_frontRibbon_00_capsule, center=<2 0 0>, rotation=<rot 0 -164.8567 -89.5001>}]
 dyn_longBow_00 [{shapeType=Capsule, length=3.0571, radius=1.1506, shapeName=dyn_longBow_00_capsule, center=<2 0 0>, rotation=<rot 0 -90.7153 -79.0209>}]
 dyn_longString_00 [{shapeType=Capsule, length=1, radius=1, shapeName=dyn_longString_00_capsule, center=<2 0 0>, rotation=<rot 90 -140.6416 -140.6417>}]
 dyn_longString_01 [{shapeType=Capsule, length=1, radius=1, shapeName=dyn_longString_01_capsule, center=<2 0 0>, rotation=<rot 90 -135.3116 -135.3118>}]
 dyn_longString_02 [{shapeType=Capsule, length=1, radius=1, shapeName=dyn_longString_02_capsule, center=<2 0 0>, rotation=<rot 90 -133.4615 -133.4617>}]
 dyn_shortBow_00 [{shapeType=Capsule, length=1.8967, radius=1.192, shapeName=dyn_shortBow_00_capsule, center=<2 0 0>, rotation=<rot 0.0001 85.7723 -88.7523>}]
 dyn_shortString_00 [{shapeType=Capsule, length=1, radius=1, shapeName=dyn_shortString_00_capsule, center=<2 0 0>, rotation=<rot 90 0 0>}]
 dyn_shortString_01 [{shapeType=Capsule, length=1, radius=1, shapeName=dyn_shortString_01_capsule, center=<2 0 0>, rotation=<rot 90 0 0>}]
 dyn_shortString_02 [{shapeType=Capsule, length=1, radius=1, shapeName=dyn_shortString_02_capsule, center=<2 0 0>, rotation=<rot 90 0 0>}]
}
constraints [
 {bone1Name=dyn_longString_00, bone2Name=base_longString_00, swing1LimitDegrees=30, swing2LimitDegrees=25, twistLimitDegrees=0}
 {bone1Name=dyn_longString_01, bone2Name=dyn_longString_00, swing1LimitDegrees=30, swing2LimitDegrees=25, twistLimitDegrees=0}
 {bone1Name=dyn_longString_02, bone2Name=dyn_longString_01, swing1LimitDegrees=30, swing2LimitDegrees=25, twistLimitDegrees=0}
 {bone1Name=dyn_shortString_00, bone2Name=base_shortString_00, swing1LimitDegrees=30, swing2LimitDegrees=25, twistLimitDegrees=0}
 {bone1Name=dyn_shortString_01, bone2Name=dyn_shortString_00, swing1LimitDegrees=30, swing2LimitDegrees=25, twistLimitDegrees=0}
 {bone1Name=dyn_shortString_02, bone2Name=dyn_shortString_01, swing1LimitDegrees=30, swing2LimitDegrees=25, twistLimitDegrees=0}
 {bone1Name=dyn_backRibbon_00, bone2Name=dyn_shortString_02, swing1LimitDegrees=10, swing2LimitDegrees=20, twistLimitDegrees=0}
 {bone1Name=dyn_frontRibbon_00, bone2Name=dyn_shortString_02, swing1LimitDegrees=10, swing2LimitDegrees=20, twistLimitDegrees=0}
 {bone1Name=dyn_longBow_00, bone2Name=dyn_shortString_02, swing1LimitDegrees=10, swing2LimitDegrees=20, twistLimitDegrees=10}
 {bone1Name=dyn_shortBow_00, bone2Name=dyn_shortString_02, swing1LimitDegrees=10, swing2LimitDegrees=20, twistLimitDegrees=10}
 {
  bone1Name=base_longString_00
  bone2Name=b_AR_Root
  swing1LimitDegrees=45
  swing1Motion=Locked
  swing2LimitDegrees=45
  swing2Motion=Locked
  twistLimitDegrees=45
  twistMotion=Locked
 }
 {
  bone1Name=base_shortString_00
  bone2Name=b_AR_Root
  swing1LimitDegrees=45
  swing1Motion=Locked
  swing2LimitDegrees=45
  swing2Motion=Locked
  twistLimitDegrees=45
  twistMotion=Locked
 }
 {bone1Name=dyn_longString_02, bone2Name=dyn_shortString_02, swing1LimitDegrees=10, swing2LimitDegrees=20, twistLimitDegrees=0}
]
