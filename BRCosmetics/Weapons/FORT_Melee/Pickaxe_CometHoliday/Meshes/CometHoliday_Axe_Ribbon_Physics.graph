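# Physics asset for the Comet Holiday pickaxe that lets its bell bows, ribbons and cable swing as secondary motion.
# Each bone has one capsule, with angularly locked constraints tying the handle to the root and the bell and cable bases to the handle.
# Two two-link bow chains and two single ribbons hang off the bell, and a two-link cable hangs off its base.
# These dynamic joints have limited swing, with zero twist limit on ribbons and cable.
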PhysicsAsset /BRCosmetics/Weapons/FORT_Melee/Pickaxe_CometHoliday/Meshes/CometHoliday_Axe_Ribbon_Physics
bodies {
 Melee_TwoHanded_Handle [{shapeType=Capsule, length=10, radius=2.5, shapeName=Melee_TwoHanded_Handle_capsule, center=<0.7897 -1.4803 48.9252>, rotation=<rot 0 -95 0>}]
 b_AR_Root [{shapeType=Capsule, length=0.505, radius=0.505, shapeName=b_AR_Root_capsule, center=<2 0 0>, rotation=<rot 0 -90 0>}]
 base_bell [{shapeType=Capsule, length=3.2277, radius=2.8483, shapeName=base_bell_capsule, center=<2 0 0>, rotation=<rot 0 90 -90>}]
 base_cable_00 [{shapeType=Capsule, length=0.505, radius=0.505, shapeName=base_cable_00_capsule, center=<2 0 0>, rotation=<rot 0 -90 0>}]
 dyn_bellBow_00 [{shapeType=Capsule, length=2.5, radius=1, shapeName=dyn_bellBow_00_capsule, center=<4.9171 0 0>}]
 dyn_bellBow_01 [{shapeType=Capsule, length=2.5, radius=1, shapeName=dyn_bellBow_01_capsule, center=<5.2073 0 0>}]
 dyn_bellBow_02 [{shapeType=Capsule, length=3.2037, radius=1.2071, shapeName=dyn_bellBow_02_capsule, center=<3.7524 -0.2884 0>}]
 dyn_bellBow_03 [{shapeType=Capsule, length=3.3975, radius=1.8198, shapeName=dyn_bellBow_03_capsule, center=<4.1022 0 -0.431>, rotation=<rot 15.0001 0 0>}]
 dyn_bellRibbon_00 [{shapeType=Capsule, length=4.5, radius=1.5, shapeName=dyn_bellRibbon_00_capsule, center=<3.0276 0 0>, rotation=<rot -0.0002 0 -0.0001>}]
 dyn_bellRibbon_01 [{shapeType=Capsule, length=4.5, radius=1.5, shapeName=dyn_bellRibbon_01_capsule, center=<3.409 0 0>, rotation=<rot 0.0002 0 0>}]
 dyn_cable_00 [{shapeType=Capsule, length=3, radius=1, shapeName=dyn_cable_00_capsule, center=<2.0454 0 -0.154>, rotation=<rot 0 -90 -80.0001>}]
 dyn_cable_01 [{shapeType=Capsule, length=3, radius=2, shapeName=dyn_cable_01_capsule, center=<2 0 0>, rotation=<rot 0 90 -84>}]
}
constraints [
 {bone1Name=dyn_bellBow_00, bone2Name=base_bell, swing1LimitDegrees=20, swing2LimitDegrees=15, twistLimitDegrees=15}
 {bone1Name=dyn_bellBow_01, bone2Name=base_bell, swing1LimitDegrees=15, swing2LimitDegrees=20, twistLimitDegrees=15}
 {bone1Name=dyn_bellRibbon_00, bone2Name=base_bell, swing1LimitDegrees=10, swing2LimitDegrees=15, twistLimitDegrees=0}
 {bone1Name=dyn_bellRibbon_01, bone2Name=base_bell, swing1LimitDegrees=10, swing2LimitDegrees=15, twistLimitDegrees=0}
 {bone1Name=dyn_cable_00, bone2Name=base_cable_00, swing1LimitDegrees=30, swing2LimitDegrees=30, twistLimitDegrees=0}
 {bone1Name=dyn_cable_01, bone2Name=dyn_cable_00, swing1LimitDegrees=40, swing2LimitDegrees=40, twistLimitDegrees=0}
 {
  bone1Name=base_bell
  bone2Name=Melee_TwoHanded_Handle
  swing1LimitDegrees=45
  swing1Motion=Locked
  swing2LimitDegrees=45
  swing2Motion=Locked
  twistLimitDegrees=45
  twistMotion=Locked
 }
 {
  bone1Name=base_cable_00
  bone2Name=Melee_TwoHanded_Handle
  swing1LimitDegrees=45
  swing1Motion=Locked
  swing2LimitDegrees=45
  swing2Motion=Locked
  twistLimitDegrees=45
  twistMotion=Locked
 }
 {
  bone1Name=Melee_TwoHanded_Handle
  bone2Name=b_AR_Root
  swing1LimitDegrees=45
  swing1Motion=Locked
  swing2LimitDegrees=45
  swing2Motion=Locked
  twistLimitDegrees=45
  twistMotion=Locked
 }
 {bone1Name=dyn_bellBow_03, bone2Name=dyn_bellBow_01, swing1LimitDegrees=15, swing2LimitDegrees=20, twistLimitDegrees=15}
 {bone1Name=dyn_bellBow_02, bone2Name=dyn_bellBow_00, swing1LimitDegrees=20, swing2LimitDegrees=20, twistLimitDegrees=15}
]
